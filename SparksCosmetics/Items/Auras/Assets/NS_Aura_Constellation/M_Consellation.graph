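# Unlit, two-sided translucent particle material for the Constellation aura effect, used on Niagara sprites and meshes.
# The constellation texture's red channel adds to its green channel, which panning refraction noise modulates for shimmer.
# A second static noise sample, thresholded by particle alpha, dissolves the shape.
# Particle color tints the result through an emissive-adjustment function, and particle alpha also scales the mask for opacity.
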Material /SparksCosmetics/Items/Auras/Assets/NS_Aura_Constellation/M_Consellation
graph {
 properties {
  bUsedWithNiagaraMeshParticles=true
  bUsedWithNiagaraSprites=true
  blendMode=BLEND_Translucent
  pixelDepthOffsetMode=PDOM_Legacy
  shadingModel=MSM_Unlit
  twoSided=true
  usedShadingModels=Unlit
 }
 nodes [
  Frac(None=Time())
  n1 = TextureSample!(UVs, Tex, Level, Apply View MipBias) {mipValueMode=TMVM_MipLevel, texture=&/T_Aura_Constellation}
  n2 = LinearInterpolate(Alpha=TextureSample!(UVs=Panner() {speedX=0.2, speedY=0.1}, Tex, Apply View MipBias) {samplerType=SAMPLERTYPE_Grayscale, texture=/Game/Environments/Apollo/Sets/IceHotel/Textures/RefractionNoise}.R) {constA=0.1}
  n3 = ParticleColor()
  n4 = Subtract(
   A=Add(A=TextureSample() {samplerType=SAMPLERTYPE_Grayscale, texture=&RefractionNoise}.R, B=@n3.A)
   B=OneMinus(None=@n3.A)
  )
  n5 = Multiply(A=Add(A=@n1.R, B=Multiply(A=@n1.G, B=@n2)), B=Saturate(None=Divide(A=@n4) {constB=0.5}))
  MaterialFunctionCall(Emissive=Multiply(A=ParticleColor().RGB, B=@n5)) {materialFunction=&MF_FX_EmissiveAdjustments}
  Multiply(A=@n5, B=@n3.A)
 ]
}
